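; Explosive gas pump prop that refills gas cans and repair torches, or lends players a tethered pump handle in Athena.
; Interaction either adds ammo and sends a refuel quest stat, or spawns or removes a held pump.
; Replicated bCableAttached drives the client-only cosmetic hose, whose component the server destroys.
; Damage and Curie interactions record the instigator, so death can spawn credited fire and fireballs when fire spread is enabled.
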(event Curie|EventOnCurieElementInteractBegun
  (params (CurieContainerHandle :type Curie%20Container%20Handle%20Structure) (ElementTag :type Gameplay%20Tag%20Structure) (InteractParamsHandle :type Curie%20Interact%20Params%20Handle%20Structure%20%28by%20ref%29))
  ;[1 1 1] Curie Elemental Interact
  (|CallParentFunction CurieContainerHandle ElementTag InteractParamsHandle)
  (bind _instigator (Curie|GetInstigatorInfofromInteractParams InteractParamsHandle))
  (bind _asfort_player_controller (Utilities|Casting|CastToFortPlayerController _instigator))
  (Variables|Default|SetExplosionInstigator (Game|Player|GetPlayerPawn _asfort_player_controller)))

(event Curie|EventOnCurieElementInteract
  (params (CurieContainerHandle :type Curie%20Container%20Handle%20Structure) (ElementTag :type Gameplay%20Tag%20Structure) (InteractParamsHandle :type Curie%20Interact%20Params%20Handle%20Structure%20%28by%20ref%29))
  ;[1 1 1] Curie Elemental Interact
  (|CallParentFunction CurieContainerHandle ElementTag InteractParamsHandle)
  (bind _instigator (Curie|GetInstigatorInfofromInteractParams InteractParamsHandle))
  (bind _asfort_player_controller (Utilities|Casting|CastToFortPlayerController _instigator))
  (Variables|Default|SetExplosionInstigator (Game|Player|GetPlayerPawn _asfort_player_controller)))

(event EventBeginPlay
  (bind (_returnvalue_x _returnvalue_y _returnvalue_z) (Math|Vector|BreakVector (Transformation|GetActorLocation)))
  ;[1 1 1] Begin Play
  (Variables|/Script/CableComponent.CableComponent|SetCableLength (select (TouchPlatform|IsMobileGame) (ScalableFloat|GetValueatLevel (Variables|Cable|GetCableLengthMobile)) (ScalableFloat|GetValueatLevel (Variables|Cable|GetCableLength))) (Variables|Default|GetCable)) ; Override Cable Length from Game Data. Only affects visuals, not actual tether radius.
  (Variables|/Script/CableComponent.CableComponent|SetSolverIterations (Math|Float|Truncate (select (TouchPlatform|IsMobileGame) (ScalableFloat|GetValueatLevel (Variables|Cable|GetCableStiffnessMobile)) (ScalableFloat|GetValueatLevel (Variables|Cable|GetCableStiffness)))) (Variables|Default|GetCable))
  (Networking|SwitchHasAuthority
    (:Authority
      (Components|DestroyComponent (Variables|Default|GetCable)) ; Destroy this component on server only since it is just for visuals and is expensive
      (Class|FortAsyncActionCurrentPlaylistReady|PlaylistDataReadyAsync
        (:OnReady
          (Variables|Default|SetCanInteractRowValue (ScalableFloat|GetValueasBoolwithContext (Variables|Default|GetCanInteract)))
          (if (and (or (Fortnite|WaterHelpers|IsActorSubmergedinOcean self) (<= _returnvalue_z (Fortnite|WaterHelpers|GetOceanTotalHeight))) (Variables|Default|GetDestroyIfSubmergedOcean))
            (Actor|SetLifeSpan (Math|Random|RandomFloatinRange 1.0 1e+01))))))
    (:Remote
      ;;[1 0 0] TODO for optimization
      (Utilities|EffectsQualityCheck 0))))

(event Custom|%28ServerOnly%29PlayRefuelFX
  ;[1 1 1] Play Refuel FX
  (GameplayCue|ExecuteGameplayCue%28Burst%29Local self (Variables|GameplayCues|GetGC_Refuel) (Ability|GameplayCue|MakeGameplayCueParameters 0.0)))

(event Interaction|EventBlueprintOnInteract
  (params (InteractingPawn :type Fort%20Pawn%20Object%20Reference) (InteractionBeingAttempted :type EInteractionBeingAttempted%20Enum :default "FirstInteraction"))
  (bind _returnvalue (FortWeapon|GetRemainingAmmo (Variables|/Script/FortniteGame.FortPawn|GetCurrentWeapon (Variables|Default|GetInteractingPlayerPawn))))
  (bind _returnvalue_2 (FortWeapon|GetBulletsPerClip (Variables|/Script/FortniteGame.FortPawn|GetCurrentWeapon (Variables|Default|GetInteractingPlayerPawn))))
  (bind _heldpetrolpump (Variables|Default|GetHeldPetrolPump))
  (bind _fortheldobject (Variables|/Game/Athena/Items/Weapons/Prototype/PetrolPump/BGA_Held_Petrol_Pump.BGA_Held_Petrol_Pump_C|GetFortHeldObject _heldpetrolpump))
  (bind _playerpawnholdingpumphandle (Variables|Default|GetPlayerPawnHoldingPumpHandle))
  (bind _interactingplayerpawn (Variables|Default|GetInteractingPlayerPawn))
  (bind _currentweapon (Variables|/Script/FortniteGame.FortPawn|GetCurrentWeapon _interactingplayerpawn))
  ;[1 1 1] On Interact
  (if (Utilities|Enum|Equal(Enum) (Fortnite|GetSubGame) "Athena")
    (bind _asfort_player_pawn (Utilities|Casting|CastToFortPlayerPawn InteractingPawn))
    (Variables|Default|SetInteractingPlayerPawn _asfort_player_pawn)
    (Utilities|IsWeaponGasCan _currentweapon
      (:Gas%20Can
        _)
      (:Not%20Gas%20Can
        (Utilities|IsWeaponRepairTorch _currentweapon
          (:True
            _)
          (:False
            (if (Utilities|NotEqual(Object) (Variables|Default|GetInteractingPlayerPawn) (Variables|Default|GetPlayerPawnHoldingPumpHandle))
              (Variables|Default|SetPlayerPawnHoldingPumpHandle (Variables|Default|GetInteractingPlayerPawn))
              (Variables|Default|SetPumpHandleInUse true)
              (if (Networking|IsServer)
                (bind _returnvalue_3 (Game|SpawnActorfromClass (Variables|Default|GetHeldPetrolPumpClass) (Transformation|GetActorTransform (Variables|Default|GetPlayerPawnHoldingPumpHandle)) "AlwaysSpawn" "SelectDefaultAtRuntime" self 0 (Variables|Default|GetPlayerPawnHoldingPumpHandle)))
                (Variables|Default|SetHeldPetrolPump _returnvalue_3)
                (Variables|/Game/Athena/Items/Weapons/Prototype/PetrolPump/BGA_Held_Petrol_Pump.BGA_Held_Petrol_Pump_C|SetPump self (Variables|Default|GetHeldPetrolPump))
                (Variables|/Game/Athena/Items/Weapons/Prototype/PetrolPump/BGA_Held_Petrol_Pump.BGA_Held_Petrol_Pump_C|SetLocation (CallFunction|GetHoseAttachStartLocation) (Variables|Default|GetHeldPetrolPump))
                (Utilities|IsValid _fortheldobject
                  (:Is%20Valid
                    (HeldObjects|Behavior|PickupHeldObject _fortheldobject _playerpawnholdingpumphandle)
                    (Utilities|IsWeaponPumpHandle (Variables|/Script/FortniteGame.FortPawn|GetCurrentWeapon _playerpawnholdingpumphandle)
                      (:Pump%20Handle
                        _)
                      (:Not%20Pump%20Handle
                        (Utilities|FlowControl|Delay 0.1)))
                    (EventDispatchers|AssignOnHolsteredEvent (Variables|Default|GetPlayerPawnHoldingPumpHandle) (delegate Custom|OnHeldPumpHolstered))
                    (CallFunction|SetCableAttachedonServer true))
                  (:Is%20Not%20Valid
                    _)) ; If pickup somehow isn't valid yet, try again.
              )
              (else
                (CallFunction|RemovePumpHandleandDetachCable)))))))
    (if (< _returnvalue _returnvalue_2)
      (Variables|Quest|SetRefillableAmmoLeft (- _returnvalue_2 _returnvalue))
      (GameplayEffects|ApplyGameplayEffectToSelf :self (Ability|GetFortAbilitySystemComponent (Variables|Default|GetInteractingPlayerPawn)) :GameplayEffectClass "/Game/Athena/Items/EnvironmentalItems/ExplosiveProps/GE_ApplyTag_RefillPetrolCan.GE_ApplyTag_RefillPetrolCan_C" :Level 0.0) ; Apply GE with tag so montage changes for refill
      (Utilities|IsWeaponRepairTorch _currentweapon
        (:True
          (FortWeapon|ModAmmo _currentweapon (Math|Float|Truncate (ScalableFloat|GetValueatLevel (Variables|Default|GetAddedFuelRepairTorch)))))
        (:False
          (FortWeapon|ModAmmo _currentweapon (Math|Float|Truncate (ScalableFloat|GetValueatLevel (Variables|Balance|GetFuelToAddToGasCanPerCycle))))))
      (CallFunction|%28ServerOnly%29PlayRefuelFX)
      (CallFunction|QuestManagerRefuelStatEvent))))

(event Custom|OnHeldPumpHolstered
  ;[1 1 1] On Interact
  (CallFunction|RemovePumpHandleandDetachCable))

(event EventOnDamageServer
  (params (Damage :type Float%20%28single-precision%29) (DamageTags :type Gameplay%20Tag%20Container%20Structure%20%28by%20ref%29) (Momentum :type Vector :default "0, 0, 0") (HitInfo :type Hit%20Result%20Structure%20%28by%20ref%29) (InstigatedBy :type Controller%20Object%20Reference) (DamageCauser :type Actor%20Object%20Reference) (EffectContext :type Gameplay%20Effect%20Context%20Handle%20Structure))
  ;[1 1 1] This is to ensure that proper kill credit is granted for pawn deaths whilst standing in curie fire
  (|CallParentFunction Damage DamageTags Momentum HitInfo InstigatedBy DamageCauser EffectContext)
  (bind _asfort_player_controller (Utilities|Casting|CastToFortPlayerController InstigatedBy)
    (:then
      (Utilities|IsValid (Game|Player|GetPlayerPawn _asfort_player_controller)
        (:Is%20Valid
          (Variables|Default|SetExplosionInstigator (Game|Player|GetPlayerPawn _asfort_player_controller)))))
    (:CastFailed
      (bind _asheld_object_interface (Utilities|Casting|CastToHeldObjectInterface DamageCauser))
      (bind _last_instigator (HeldObject|GetLastInstigator _asheld_object_interface))
      (Utilities|IsValid _last_instigator
        (:Is%20Valid
          (Utilities|IsValid (Game|Player|GetPlayerPawn _last_instigator)
            (:Is%20Valid
              (Variables|Default|SetExplosionInstigator (Game|Player|GetPlayerPawn _last_instigator)))))))))

(event EventOnDeathServer
  (params (Damage :type Float%20%28single-precision%29) (DamageTags :type Gameplay%20Tag%20Container%20Structure%20%28by%20ref%29) (Momentum :type Vector :default "0, 0, 0") (HitInfo :type Hit%20Result%20Structure%20%28by%20ref%29) (InstigatedBy :type Controller%20Object%20Reference) (DamageCauser :type Actor%20Object%20Reference) (EffectContext :type Gameplay%20Effect%20Context%20Handle%20Structure))
  ;[1 1 1] Curie related. On death spawn a BGA fire patch
  (|CallParentFunction Damage DamageTags Momentum HitInfo InstigatedBy DamageCauser EffectContext)
  (CallFunction|RemovePumpHandleandDetachCable)
  (if (and (Curie|IsCurieActiveforActor self) (and (ScalableFloat|GetValueasBool (Variables|Default|GetRow_SpreadFire)) (Variables|Default|GetbSpreadFire)))
    (Building|SpawnBuildingGameplayActor (Variables|Default|GetGasPumpFireMesh) (Math|Transform|MakeTransform (Transformation|GetWorldLocation (Variables|Debug|GetStaticMeshComponent)) (Transformation|GetWorldRotation (Variables|Debug|GetStaticMeshComponent))) (Variables|Default|GetExplosionInstigator))
    ;[1 1 1] Fire out projectiles at pseudo random angles & speeds.
    (for _index (range (Math|Float|Truncate (ScalableFloat|GetValueatLevel (Variables|Balance|Fireballs|GetFireBallsToSPawn))))
      (Projectile|SpawnProjectile :ProjectileClass "/Game/Athena/Items/EnvironmentalItems/ExplosiveProps/B_Prj_Athena_GasPump_Fireballs.B_Prj_Athena_GasPump_Fireballs_C" :RequestedBy (select (Utilities|IsValid (Variables|Default|GetExplosionInstigator) :!class /Script/Engine.KismetSystemLibrary) (Variables|Default|GetExplosionInstigator) self) :SpawnLocation (+ (Transformation|GetWorldLocation (Variables|Debug|GetStaticMeshComponent)) "0, 0,25.000000") :SpawnRotation "0, 0, 0" :SpawnDirection (Math|Rotator|MakeRotfromX (Math|Random|RandomUnitVectorinConeinDegrees (+ (Utilities|Array|Get(acopy) (Variables|Balance|Fireballs|GetProjectileSpawnVectors) _index) (Transformation|GetUpVector (Variables|Debug|GetStaticMeshComponent))) (ScalableFloat|GetValueatLevel (Variables|Balance|Fireballs|GetConeHalfAngleInDegrees)))) :bUseDefaultPhysics false :InitialSpeed (Math|Random|RandomFloatinRange (ScalableFloat|GetValueatLevel (Variables|Balance|Fireballs|GetSpawnedProjectileMinSpeed)) (ScalableFloat|GetValueatLevel (Variables|Balance|Fireballs|GetSpawnedProjectileMaxSpeed))) :GravityScale 1.25 :ChargePercent 0.0))))

(event Custom|AttachCable
  (bind _playerpawnholdingpumphandle (Variables|Default|GetPlayerPawnHoldingPumpHandle))
  (bind _currentweapon (Variables|/Script/FortniteGame.FortPawn|GetCurrentWeapon _playerpawnholdingpumphandle))
  ;[1 1 1] Attach Cable
  (Networking|SwitchHasAuthority
    (:Remote
      (if (Variables|Default|GetbCableAttached)
        (Utilities|IsValid _playerpawnholdingpumphandle
          (:Is%20Valid
            (Utilities|IsValid _currentweapon
              (:Is%20Valid
                (Utilities|IsWeaponPumpHandle _currentweapon
                  (:Pump%20Handle
                    (Rendering|SetVisibility (Variables|Default|GetS_Commerce_GasPump_02) false false)
                    (Utilities|IsValid (Variables|Default|GetCable)
                      (:Is%20Valid
                        (Transformation|SetWorldLocation (Variables|Default|GetCable) (CallFunction|GetHoseAttachStartLocation) false true)
                        (Rendering|SetVisibility (Variables|Default|GetCable) true false)
                        (Cable|SetAttachEndTo (Variables|Default|GetCable) (Variables|/Script/FortniteGame.FortPawn|GetCurrentWeapon (Variables|Default|GetPlayerPawnHoldingPumpHandle)) "None" (Variables|Cable|GetSocketName_CableAttachEnd))
                        (Variables|/Script/CableComponent.CableComponent|SetbAttachEnd true (Variables|Default|GetCable)))))
                  (:Not%20Pump%20Handle
                    _)))
              (:Is%20Not%20Valid
                _))
            (Utilities|FlowControl|Delay 0.1)
            (CallFunction|AttachCable))
          (:Is%20Not%20Valid
            _))))))

(event Custom|SetCableAttachedOnServer
  (params (bAttached :type Boolean))
  ;[1 1 1] Attach Cable
  (Networking|SwitchHasAuthority
    (:Authority
      (Variables|Default|SetbCableAttached bAttached)
      (Networking|ForceNetUpdate))))

(event Custom|DetachCable
  ;[1 1 1] Detach Cable
  (Variables|Default|SetPumpHandleInUse false)
  (Variables|Default|SetPlayerPawnHoldingPumpHandle 0)
  (Networking|SwitchHasAuthority
    (:Remote
      (Rendering|SetVisibility (Variables|Default|GetS_Commerce_GasPump_02) true false)
      (GameplayCue|ExecuteGameplayCue%28Burst%29Local self (Variables|GameplayCues|GetGC_PumpRetractCable) (Ability|GameplayCue|MakeGameplayCueParameters :NormalizedMagnitude 0.0 :RawMagnitude 0.0 :EffectContext (Utilities|Struct|MakeGameplayEffectContextHandle) :Location "0,0,0" :Normal "0,0,0" :Instigator self :EffectCauser self :SourceObject self :GameplayEffectLevel 1 :AbilityLevel 1 :bReplicateLocationWhenUsingMinimalRepProxy false))
      (Utilities|IsValid (Variables|Default|GetCable)
        (:Is%20Valid
          (Networking|ForceNetUpdate self)
          (Rendering|SetVisibility (Variables|Default|GetCable) false false)
          (Cable|SetAttachEndTo (Variables|Default|GetCable) 0 "None" "None")
          (Variables|/Script/CableComponent.CableComponent|SetbAttachEnd false (Variables|Default|GetCable)))))))

(event Custom|RemovePumpHandleAndDetachCable
  (bind _heldpetrolpump (Variables|Default|GetHeldPetrolPump))
  (bind _playerpawnholdingpumphandle (Variables|Default|GetPlayerPawnHoldingPumpHandle))
  ;[1 1 1] Remove Pump Handle and Detach Cable
  (Networking|SwitchHasAuthority
    (:Authority
      (if (and (Variables|Default|GetPumpHandleInUse) (Utilities|Enum|Equal(Enum) (Fortnite|GetSubGame) "Athena"))
        (Utilities|FlowControl|Sequence
          (:then_0
            (Utilities|IsValid _playerpawnholdingpumphandle
              (:Is%20Valid
                (EventDispatchers|UnbindEventfromOnHolsteredEvent _playerpawnholdingpumphandle (EventDispatchers|CreateEvent :!event Custom|OnHeldPumpHolstered)))))
          (:then_1
            (CallFunction|SetCableAttachedonServer false))
          (:then_2
            (Utilities|IsValid _heldpetrolpump
              (:Is%20Valid
                (Actor|DestroyActor _heldpetrolpump)))))))))

(fn BlueprintCanInteract
  (params (InteractingPawn :type Fort%20Pawn%20Object%20Reference) (InteractionBeingAttempted :type EInteractionBeingAttempted%20Enum :default "FirstInteraction") (InteractionType :type TInteractionType%20Enum :default "IT_NoInteraction"))
  (if (not (GameplayTags|HasAnyMatchingGameplayTags InteractingPawn (Variables|Default|GetTC_BlockInteract)))
    (if (and (Utilities|Enum|Equal(Enum) (Fortnite|GetSubGame) "Athena") (Variables|Default|GetCanInteractRowValue))
      (bind _asfort_player_pawn (Utilities|Casting|CastToFortPlayerPawn InteractingPawn))
      (if (GameplayTags|HasMatchingGameplayTag _asfort_player_pawn "(TagName=\"GameplayEffect.ForagedItem.SpookyMist\")")
        (return false)
        (else
          (Utilities|IsWeaponRepairTorch (Variables|/Script/FortniteGame.FortPawn|GetCurrentWeapon _asfort_player_pawn)
            (:True
              _)
            (:False
              (Utilities|IsWeaponGasCan (Variables|/Script/FortniteGame.FortPawn|GetCurrentWeapon _asfort_player_pawn)
                (:Gas%20Can
                  _)
                (:Not%20Gas%20Can
                  (Utilities|IsWeaponPumpHandle (Variables|/Script/FortniteGame.FortPawn|GetCurrentWeapon _asfort_player_pawn)
                    (:Pump%20Handle
                      (return (Utilities|Equal(Object) (Variables|Default|GetPlayerPawnHoldingPumpHandle) _asfort_player_pawn)))
                    (:Not%20Pump%20Handle
                      (bind _returnvalue (Fortnite|HeldObjects|GetHeldObjectAttachedtoPlayer _asfort_player_pawn))
                      (return (and (not (Utilities|IsValid _returnvalue :!class /Script/Engine.KismetSystemLibrary)) (not (Variables|Default|GetPumpHandleInUse))))))))))
          (return true)))
      (else
        (return false)))
    (else
      (return false))))

(fn BlueprintGetInteractionString
  (params (InteractingPawn :type Fort%20Pawn%20Object%20Reference) (InteractionBeingAttempted :type EInteractionBeingAttempted%20Enum :default "FirstInteraction"))
  (bind _currentweapon (Variables|/Script/FortniteGame.FortPawn|GetCurrentWeapon InteractingPawn))
  (Utilities|IsWeaponGasCan _currentweapon
    (:Gas%20Can
      (return "NSLOCTEXT(\"[08D7F1AE428B09F35F44C7892DBAF4AE]\", \"98745B3E4945BA5E2FCCBCA13FE057E3\", \"Refill Gas Can\")"))
    (:Not%20Gas%20Can
      (Utilities|IsWeaponRepairTorch _currentweapon
        (:True
          (return "NSLOCTEXT(\"[08D7F1AE428B09F35F44C7892DBAF4AE]\", \"D2EA65E44C2DE0592F66E88A887F7741\", \"Refill Repair Torch\")"))
        (:False
          (if (Utilities|Equal(Object) InteractingPawn (Variables|Default|GetPlayerPawnHoldingPumpHandle))
            (return "NSLOCTEXT(\"[08D7F1AE428B09F35F44C7892DBAF4AE]\", \"A51023D84C19660F40974DBDE3E2A4A7\", \"Drop Pump\")")
            (else
              (return "NSLOCTEXT(\"[08D7F1AE428B09F35F44C7892DBAF4AE]\", \"AE7C74704123C2A5C0DBAA8EF72EF6DF\", \"Grab Pump\")"))))))))

(fn BlueprintGetInteractionTime
  (params (InteractingPawn :type Fort%20Pawn%20Object%20Reference) (InteractionBeingAttempted :type EInteractionBeingAttempted%20Enum :default "FirstInteraction"))
  (bind _currentweapon (Variables|/Script/FortniteGame.FortPawn|GetCurrentWeapon InteractingPawn))
  (Utilities|IsWeaponGasCan _currentweapon
    (:Gas%20Can
      _)
    (:Not%20Gas%20Can
      (Utilities|IsWeaponRepairTorch _currentweapon
        (:True
          (if (< (FortWeapon|GetRemainingAmmo _currentweapon) (FortWeapon|GetBulletsPerClip _currentweapon))
            (return (ScalableFloat|GetValueatLevel (Variables|Balance|GetPetrolPickupInteractTime)) true)
            (else
              _)))
        (:False
          _))
      (return (ScalableFloat|GetValueatLevel (Variables|Balance|GetDefaultInteractTime)) true))))

(fn BlueprintOnLocalInteract
  (params (InteractingPawn :type Fort%20Player%20Pawn%20Object%20Reference) (InteractionBeingAttempted :type EInteractionBeingAttempted%20Enum :default "FirstInteraction"))
  (bind _currentweapon (Variables|/Script/FortniteGame.FortPawn|GetCurrentWeapon InteractingPawn))
  (Utilities|IsWeaponGasCan _currentweapon
    (:Gas%20Can
      (if (>= (FortWeapon|GetRemainingAmmo _currentweapon) (FortWeapon|GetBulletsPerClip _currentweapon))
        (FailedAction|UnabletoPerformAction InteractingPawn true (Variables|Default|GetTag_FailedInteract_PetrolPickupIsFull))
        (Audio|PlaySoundatLocation (Variables|Default|GetFailedInteractSound) (Transformation|GetActorLocation))
        (return true)
        (else
          (return false))))
    (:Not%20Gas%20Can
      (return false))))

(fn BuildDataRegistryResolverScope
  ;[1 0.8475 0.2776] Gives the name of the volume as a scope to the data registry. Needed for calls to GetValueWithContext to work
  (return (Utilities|Array|MakeArray (SpatialGameplay|GetSpatialGameplayInterfaceFNameforActor (Minigame|GetMinigame self :!class /Script/FortniteGame.MinigameBlueprintLibrary))) 0 true))

(fn GetCacheContextOverride
  (return 0))

(fn GetHoseAttachStartLocation
  (return (Math|Vector|SelectVector (Transformation|GetSocketLocation (Variables|Debug|GetStaticMeshComponent) (Variables|Cable|GetSocketName_CableAttachStart)) (Transformation|GetActorLocation) (Transformation|DoesSocketExist (Variables|Debug|GetStaticMeshComponent) (Variables|Cable|GetSocketName_CableAttachStart)))))

(fn LocalOnFailedInteract
  (params (InteractingPawn :type Fort%20Player%20Pawn%20Object%20Reference)))

(fn OnRep_bCableAttached
  (if (Variables|Default|GetbCableAttached)
    (CallFunction|AttachCable)
    (else
      (CallFunction|DetachCable))))

(fn QuestManagerRefuelStatEvent
  (bind _interactingplayerpawn (Variables|Default|GetInteractingPlayerPawn))
  (bind _currentweapon (Variables|/Script/FortniteGame.FortPawn|GetCurrentWeapon _interactingplayerpawn))
  ;[1 1 1] all Refuel stat events should be SourceTags=Actors tags that caused the refuel, TargetTags=Actor/item that was refueled.
  ;
  ; Quests will want to know the exact amount of fuel we added. We have no way to get a callback from ModAmmo()  to see how much we ended up giving. So we use RefillableAmmoLeft to infer the correct amount that the ModAmmo() call gave. Otherwise if the player is at 99/100 fuel ammo in an item we would give a progress count of 10 when it should be 1.
  ;
  ;
  (Utilities|IsWeaponGasCan _currentweapon
    (:Gas%20Can
      (LocalVariables|SetbIsWeaponGasCan true))
    (:Not%20Gas%20Can
      (LocalVariables|SetbIsWeaponGasCan false)))
  (Quest|SendCustomStatEventwithTags :self (Quests|GetQuestManager (Fortnite|GetFortPlayerControllerfromActor (Variables|Default|GetInteractingPlayerPawn)) "Athena") :Type "Refuel" :AdditionalSourceTags (Tags|GetWeaponDataGameplayTags _currentweapon) :Count (select (< (Variables|Quest|GetRefillableAmmoLeft) (select (LocalVariables|GetbIsWeaponGasCan) (Math|Float|Truncate (ScalableFloat|GetValueatLevel (Variables|Balance|GetFuelToAddToGasCanPerCycle))) (Math|Float|Truncate (ScalableFloat|GetValueatLevel (Variables|Default|GetAddedFuelRepairTorch))))) (Variables|Quest|GetRefillableAmmoLeft) (select (LocalVariables|GetbIsWeaponGasCan) (Math|Float|Truncate (ScalableFloat|GetValueatLevel (Variables|Balance|GetFuelToAddToGasCanPerCycle))) (Math|Float|Truncate (ScalableFloat|GetValueatLevel (Variables|Default|GetAddedFuelRepairTorch)))))))

(fn ConstructionScript
  (|CallParentFunction))
