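; Explosive building prop Blueprint that catches fire from damage or Curie fire, then detonates after a randomized fuse, crediting the last instigator.
; Clamped damage and a ShouldDie override delay death until detonation, while gas-pump chains get staggered timers and heavy explosive hits a fast one.
; Non-silent death runs a throttled async sphere overlap that damages and launches actors, skipped in idle minigames.
; Explosion FX play through a RepNotify because reliable multicasts proved unreliable.
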(event EventBeginPlay
  (|CallParentFunction))

(event Custom|ExplodeAndDieStaggered
  ;[1 1 1] If Destroyed by Anything but PickAxe, Explode
  ;;[1 0.4346 0] Special Case for handling chain reaction of explosions, to stagger them to ensure we do not simply delete a player than is too close (Jason West Request)
  (CallFunction|GrowTimeline)
  (Rendering|SetActorHiddenInGame true)
  (Variables|Default|SetbShouldDie true))

(event Custom|TriggeredByGasPump
  ;[1 1 1] If Destroyed by Anything but PickAxe, Explode
  ;;[1 0.4346 0] Special Case for handling chain reaction of explosions, to stagger them to ensure we do not simply delete a player than is too close (Jason West Request)
  (Utilities|Time|SetTimerbyEvent (delegate Custom|ExplodeAndDieStaggered) (Math|Random|RandomFloatinRange 1.7 4.5))
  (CallFunction|ManuallyTriggerFuse))

(event EventOnDeathServer
  (params (Damage :type Float%20%28single-precision%29) (DamageTags :type Gameplay%20Tag%20Container%20Structure%20%28by%20ref%29) (Momentum :type Vector :default "0, 0, 0") (HitInfo :type Hit%20Result%20Structure%20%28by%20ref%29) (InstigatedBy :type Controller%20Object%20Reference) (DamageCauser :type Actor%20Object%20Reference) (EffectContext :type Gameplay%20Effect%20Context%20Handle%20Structure))
  ;[1 1 1] If Destroyed by Anything but PickAxe, Explode
  ;;[1 1 1] On Death Logic
  (bind _returnvalue (MissionHelpers|HasTags DamageTags (GameplayTags|MakeGameplayTagContainerfromArray (Utilities|Array|MakeArray "(TagName=\"Gameplay.Effect.InstantDeath.Silent\")" "(TagName=\"Gameplay.Effect.InstantDeath.StructuralSupport\")" "(TagName=\"Gameplay.Effect.InstantDeath.PersistentUpdate\")" "(TagName=\"Gameplay.Effect.InstantDeath.Environment\")" "(TagName=\"Gameplay.Effect.InstantDeath.Detachment\")" "(TagName=\"Gameplay.Effect.InstantDeath.Despawn\")" "(TagName=\"Gameplay.Effect.InstantDeath.Culled\")")) false false))
  (if (not _returnvalue)
    (CallFunction|ExplosionTrace)
    ;;;[1 1 1] This is because a reliable multicast wasn't getting called everytime. Todd Eckert, Dalvin Kang.
    (Variables|Default|SetExplosiveFxRepNotify true)))

(event EventOnDamageServer
  (params (Damage :type Float%20%28single-precision%29) (DamageTags :type Gameplay%20Tag%20Container%20Structure%20%28by%20ref%29) (Momentum :type Vector :default "0, 0, 0") (HitInfo :type Hit%20Result%20Structure%20%28by%20ref%29) (InstigatedBy :type Controller%20Object%20Reference) (DamageCauser :type Actor%20Object%20Reference) (EffectContext :type Gameplay%20Effect%20Context%20Handle%20Structure))
  (bind _asfort_player_controller (Utilities|Casting|CastToFortPlayerController InstigatedBy)
    (:then
      (Variables|Default|SetLastDamageInstigator _asfort_player_controller))
    (:CastFailed
      _))
  ;[1 1 1] If Destroyed by Anything but PickAxe, Explode
  (Utilities|IsA(soft) DamageCauser "/Game/Athena/Items/EnvironmentalItems/ExplosiveProps/Apollo_GasPump_Valet.Apollo_GasPump_Valet_C"
    (:True
      (Utilities|FlowControl|Sequence
        (:then_0
          (CallFunction|TriggeredbyGasPump))
        (:then_1
          _)))
    (:False
      (if (and (>= Damage 1e+02) (GameplayTags|HasAnyTags DamageTags (Variables|Default|GetExplosiveDamageTag) true))
        ;[1 0.3551 0] If the damage was Explosive and dealt more than 100 damage, we kill much faster than normal
        (Utilities|Time|SetTimerbyEvent (delegate Custom|FastExplodeAndDie) 0.25)
        (else
          _))))
  (Utilities|FlowControl|Gate false
    (:Exit
      (Utilities|FlowControl|Sequence
        (:then_0
          (if (Variables|Bools|GetHealthIsZero)
            _))
        (:then_1
          ;;[1 1 1] Set On Fire
          (if (<= (Health|GetHealthPercent) (Variables|Default|GetOnFireThreshold))
            (Utilities|FlowControl|DoOnce false
              (:Completed
                (Utilities|IsValid (Variables|Default|GetLastDamageInstigator)
                  (:Is%20Valid
                    (bind _returnvalue (GameplayEffects|FortApplyGameplayEffectToTarget :self (Ability|GetFortAbilitySystemComponent (Pawn|GetControlledPawn (Variables|Default|GetLastDamageInstigator))) :GameplayEffectClass "/Game/Athena/Items/EnvironmentalItems/ReactiveProps/GE_Athena_ReactiveProps_OnFire.GE_Athena_ReactiveProps_OnFire_C" :TargetOfEffect (Ability|GetFortAbilitySystemComponent self) :OptionalSourceObject (Pawn|GetControlledPawn (Variables|Default|GetLastDamageInstigator)) :Level 0.0))
                    (Variables|Default|SetOnFireEffect _returnvalue))
                  (:Is%20Not%20Valid
                    (bind _returnvalue_2 (Environment|ApplyGlobalEnvironmentGameplayEffecttoActor self "/Game/Athena/Items/EnvironmentalItems/ReactiveProps/GE_Athena_ReactiveProps_OnFire.GE_Athena_ReactiveProps_OnFire_C"))
                    (Variables|Default|SetOnFireEffect _returnvalue_2)))
                (CallFunction|OnFireFx))))))
      (CallFunction|StartDetonation))))

(event Custom|ExplosionTrace
  ;[1 0.2249 0.04801] Explosive Damage Logic
  (Utilities|FlowControl|DoOnce false
    (:Completed
      (bind _returnvalue (Utility|OverlapQueue|GetAsyncOverlapQueue self))
      (Class|AsyncOverlapQueue|ThrottledAsyncSphereOverlapActors :self _returnvalue :Location (+ (Transformation|GetActorLocation) (Variables|Default|GetExplosionHeightOffset)) :Radius (ScalableFloat|GetValueatLevel (Variables|Default|GetRadius)) :bCheckLineOfSight false :ShouldCancelProcess (EventDispatchers|CreateEvent :!event ShouldCancelExplosion) :ProcessHitActorFunction (delegate Custom|DestroyHitActors) :CompletionFunction (delegate Custom|OnCompletion) :ObjectTypes "ObjectTypeQuery1" :TraceChannel "ECC_Visibility"))))

(event Custom|FastExplodeAndDie
  ;[1 1 1] If Destroyed by Anything but PickAxe, Explode
  ;[1 0.3551 0] If the damage was Explosive and dealt more than 100 damage, we kill much faster than normal
  (Development|SetHiddeninGame (Variables|Debug|GetStaticMeshComponent) true false)
  (Gameplay|Die "(TagName=\"Gameplay.Effect.InstantDeath\")" (stale EventOnDamageServer InstigatedBy) (stale EventOnDamageServer DamageCauser)) ; DestroyActor credits Self with destruction and \nresulting explosion for quest purposes. \nTo preserve instigator use Die Instead.
)

(event Custom|OnFireFx
  ;[1 1 1] If Destroyed by Anything but PickAxe, Explode
  ;;[1 1 1] Set On Fire
  (Projectile|AddGameplayCue%28Looping%29Local self (Variables|Default|GetFireCue) (Ability|GameplayCue|MakeGameplayCueParameters 0.0)))

(event Custom|ManuallyTriggerFuse
  ;[1 1 1] If Destroyed by Anything but PickAxe, Explode
  ;;[1 1 1] Set On Fire
  (Utilities|FlowControl|DoOnce false
    (:Completed
      (Utilities|IsValid (Variables|Default|GetLastDamageInstigator)
        (:Is%20Valid
          (bind _returnvalue (GameplayEffects|FortApplyGameplayEffectToTarget :self (Ability|GetFortAbilitySystemComponent (Pawn|GetControlledPawn (Variables|Default|GetLastDamageInstigator))) :GameplayEffectClass "/Game/Athena/Items/EnvironmentalItems/ReactiveProps/GE_Athena_ReactiveProps_OnFire.GE_Athena_ReactiveProps_OnFire_C" :TargetOfEffect (Ability|GetFortAbilitySystemComponent self) :OptionalSourceObject (Pawn|GetControlledPawn (Variables|Default|GetLastDamageInstigator)) :Level 0.0))
          (Variables|Default|SetOnFireEffect _returnvalue))
        (:Is%20Not%20Valid
          (bind _returnvalue_2 (Environment|ApplyGlobalEnvironmentGameplayEffecttoActor self "/Game/Athena/Items/EnvironmentalItems/ReactiveProps/GE_Athena_ReactiveProps_OnFire.GE_Athena_ReactiveProps_OnFire_C"))
          (Variables|Default|SetOnFireEffect _returnvalue_2)))
      (CallFunction|OnFireFx)
      (CallFunction|StartDetonation))))

(event EventOnDeathPlayEffects
  (params (Damage :type Float%20%28single-precision%29) (DamageTags :type Gameplay%20Tag%20Container%20Structure%20%28by%20ref%29) (Momentum :type Vector :default "0, 0, 0") (HitInfo :type Hit%20Result%20Structure%20%28by%20ref%29) (InstigatedBy :type Fort%20Pawn%20Object%20Reference) (DamageCauser :type Actor%20Object%20Reference) (EffectContext :type Gameplay%20Effect%20Context%20Handle%20Structure))
  ;[1 1 1] If Destroyed by Anything but PickAxe, Explode
  (CallFunction|StopGrowAudio))

(event Custom|ResetFuse
  ;[1 1 1] If Destroyed by Anything but PickAxe, Explode
  ;;[1 1 1] Set On Fire
  (Utilities|FlowControl|DoOnce false
    (:Completed
      (Utilities|IsValid (Variables|Default|GetLastDamageInstigator)
        (:Is%20Valid
          (bind _returnvalue (GameplayEffects|FortApplyGameplayEffectToTarget :self (Ability|GetFortAbilitySystemComponent (Pawn|GetControlledPawn (Variables|Default|GetLastDamageInstigator))) :GameplayEffectClass "/Game/Athena/Items/EnvironmentalItems/ReactiveProps/GE_Athena_ReactiveProps_OnFire.GE_Athena_ReactiveProps_OnFire_C" :TargetOfEffect (Ability|GetFortAbilitySystemComponent self) :OptionalSourceObject (Pawn|GetControlledPawn (Variables|Default|GetLastDamageInstigator)) :Level 0.0))
          (Variables|Default|SetOnFireEffect _returnvalue))
        (:Is%20Not%20Valid
          (bind _returnvalue_2 (Environment|ApplyGlobalEnvironmentGameplayEffecttoActor self "/Game/Athena/Items/EnvironmentalItems/ReactiveProps/GE_Athena_ReactiveProps_OnFire.GE_Athena_ReactiveProps_OnFire_C"))
          (Variables|Default|SetOnFireEffect _returnvalue_2)))
      (CallFunction|OnFireFx)
      (CallFunction|StartDetonation))))

(event Custom|DestroyHitActors
  (params (HitActor :type Actor%20Object%20Reference))
  ;[1 0.2249 0.04801] Explosive Damage Logic
  ;; Skip LOS check on player built buildings
  (bind _asbuilding_static_mesh (Utilities|Casting|CastToBuildingSMActor HitActor)
    (:then
      (if (not (BuildingActor|IsPlayerBuilt _asbuilding_static_mesh))
        (bind _in_line_of_sight (Class|MissionBlueprintFunctionLibrary|CheckBasicActorLineOfSight self (Variables|Default|GetExplosionTraceOffset) HitActor "0, 0, 0" false "TraceTypeQuery5" "Persistent" 15.0 true :!label _label))
        (if _in_line_of_sight
          _))
      (CallFunction|ExecuteExplosiononActor HitActor))
    (:CastFailed
      _))
  (goto _label))

(event Custom|OnCompletion
  (params (WasCancelled :type Boolean)))

(event Custom|StartDetonation
  ;[0.4792 0.2571 0] Start Detonation
  ; and Explode And Die
  (bind _returnvalue (CallFunction|DoesMutatorBlockIncomingDamage))
  (if (not _returnvalue)
    (Utilities|FlowControl|DoOnce false
      (:Completed
        (CallFunction|GrowTimeline)
        (Variables|Default|SetCanGenerateResources false)
        (Utilities|Time|SetTimerbyEvent (delegate Custom|ExplodeAndDie) (Math|Random|RandomFloatinRange 1.7 3.0))))))

(event Custom|ResetDetonation
  ;[0.4792 0.2571 0] Start Detonation
  ; and Explode And Die
  (Utilities|FlowControl|DoOnce false
    (:Completed
      (CallFunction|GrowTimeline)
      (Variables|Default|SetCanGenerateResources false)
      (Utilities|Time|SetTimerbyEvent (delegate Custom|ExplodeAndDie) (Math|Random|RandomFloatinRange 1.7 3.0)))))

(event Custom|ExplodeAndDie
  ;[0.4792 0.2571 0] Start Detonation
  ; and Explode And Die
  (Health|SetHealth 0.0)
  (Rendering|SetActorHiddenInGame true)
  (Variables|Default|SetbShouldDie true))

(event Custom|ResetDetonationSequence
  (CallFunction|ResetFuse)
  (CallFunction|ResetDetonation))

(event Custom|GrowTimeline
  ;[1 1 1] Pre Explode Grow Timeline
  (Audio|Components|Audio|Play (Variables|Default|GetPreExplodeAudioComponent) 0.0))

(event Custom|StopGrowAudio
  ;[1 1 1] Pre Explode Grow Timeline
  (bind _preexplodeaudiocomponent (Variables|Default|GetPreExplodeAudioComponent)
    (:then
      (if (Audio|Components|Audio|IsPlaying _preexplodeaudiocomponent)
        (Audio|Components|Audio|FadeOut _preexplodeaudiocomponent 0.25 0.0 "Linear" :!class /Script/Engine.AudioComponent)))
    (:else)))

(event Custom|MultiDeactivateEmitter
  ;[1 1 1] Deactivate attached spawned emitters on death (to avoid having them linger for a second, they destroy automatically)
  (if (not (Networking|IsServer))
    (Projectile|RemoveGameplayCue%28Looping%29Local self (Variables|Default|GetFireCue) (Ability|GameplayCue|MakeGameplayCueParameters 0.0))
    (for _array_element (Variables|DamageBasedEmitters|GetSpawnedDamageStateEmitters)
      (Components|Activation|Deactivate _array_element))))

(event Curie|EventOnCurieElementInteract
  (params (CurieContainerHandle :type Curie%20Container%20Handle%20Structure) (ElementTag :type Gameplay%20Tag%20Structure) (InteractParamsHandle :type Curie%20Interact%20Params%20Handle%20Structure%20%28by%20ref%29))
  ;[1 1 1] This allows Curie to trigger the damage over time state
  (if (GameplayTags|Equal(GameplayTag) ElementTag "(TagName=\"Curie.Element.Fire\")")
    ;;[1 0 0.2634] This is to ensure kill credit is correctly passed through curie fire
    (bind _instigator (Curie|GetInstigatorInfofromInteractParams InteractParamsHandle))
    (Utilities|IsValid (Fortnite|GetFortPlayerControllerfromActor _instigator)
      (:Is%20Valid
        (Variables|Default|SetLastDamageInstigator (Fortnite|GetFortPlayerControllerfromActor _instigator)))
      (:Is%20Not%20Valid
        _))
    (CallFunction|ManuallyTriggerFuse)))

(event Curie|EventOnCurieElementInteractBegun
  (params (CurieContainerHandle :type Curie%20Container%20Handle%20Structure) (ElementTag :type Gameplay%20Tag%20Structure) (InteractParamsHandle :type Curie%20Interact%20Params%20Handle%20Structure%20%28by%20ref%29))
  ;[1 1 1] This allows Curie to trigger the damage over time state
  (if (GameplayTags|Equal(GameplayTag) ElementTag "(TagName=\"Curie.Element.Fire\")")
    ;;[1 0 0.2634] This is to ensure kill credit is correctly passed through curie fire
    (bind _instigator (Curie|GetInstigatorInfofromInteractParams InteractParamsHandle))
    (Utilities|IsValid (Fortnite|GetFortPlayerControllerfromActor _instigator)
      (:Is%20Valid
        (Variables|Default|SetLastDamageInstigator (Fortnite|GetFortPlayerControllerfromActor _instigator)))
      (:Is%20Not%20Valid
        _))
    (CallFunction|ManuallyTriggerFuse)))

(fn ApplyDamageToActor
  (params (Actor%20to%20Damage :type Actor%20Object%20Reference))
  (bind _lastdamageinstigator (Variables|Default|GetLastDamageInstigator))
  (Utilities|IsValid _lastdamageinstigator
    (:Is%20Valid
      (GameplayEffects|FortApplyGameplayEffectToTarget :self (Ability|GetFortAbilitySystemComponent (Pawn|GetControlledPawn _lastdamageinstigator)) :GameplayEffectClass (Variables|GameplayEffects|GetGE_Explode) :TargetOfEffect (Ability|GetFortAbilitySystemComponent (Variables|GetActor%20to%20Damage)) :OptionalSourceObject self :OptionalEffectCauser self :OptionalDamageSource self :Level 0.0))
    (:Is%20Not%20Valid
      (bind _asfort_pawn (Utilities|Casting|CastToFortPawn (Variables|GetActor%20to%20Damage))
        (:then
          (Utilities|IsinVehicle _asfort_pawn
            (:false
              _)))
        (:CastFailed
          (Utilities|Casting|CastToFortVehicleInterface (Variables|GetActor%20to%20Damage)
            (:then
              (Environment|ApplyGlobalEnvironmentDamagetoActor (Variables|GetActor%20to%20Damage) (ScalableFloat|GetValueatLevel (Variables|Default|GetVehicleDamage))))
            (:CastFailed
              _))))
      (Environment|ApplyGlobalEnvironmentGameplayEffecttoActor (Variables|GetActor%20to%20Damage) (Variables|GameplayEffects|GetGE_Explode)))))

(fn CanAttemptGenerateResources
  (params (InTags :type Gameplay%20Tag%20Container%20Structure%20%28by%20ref%29) (InstigatorController :type Controller%20Object%20Reference))
  (return (Variables|Default|GetCanGenerateResources)))

(fn BlueprintModifyIncomingDamage
  (params (Damage :type Float%20%28single-precision%29) (InTags :type Gameplay%20Tag%20Container%20Structure%20%28by%20ref%29) (EffectContext :type Gameplay%20Effect%20Context%20Handle%20Structure%20%28by%20ref%29) (EventInstigator :type Controller%20Object%20Reference) (DamageCauser :type Actor%20Object%20Reference))
  ;[1 1 1] Clamp incoming damage to Current Health - 1 to prevent visual death/destruction/hiding before explosion
  (return (Math|Float|SelectFloat Damage (- (Health|GetHealth) 1.0) (< Damage (Health|GetHealth)))))

(fn ExecuteExplosionOnActor
  (params (ActorToExplode :type Actor%20Object%20Reference))
  (CallFunction|ApplyDamagetoActor ActorToExplode)
  (CallFunction|LaunchActor ActorToExplode)
  (Development|PrintString (Utilities|GetDisplayName ActorToExplode) true true "(R=1.000000,G=0.044237,B=0.000000,A=1.000000)" 5.0))

(fn ExecuteFireApplicationInRadiusIfCurieEnabled
  (params (FireApplicationTarget :type Actor%20Object%20Reference))
  (bind _self self)
  (Networking|SwitchHasAuthority
    (:Authority
      (if (Curie|IsCurieActive)
        (Utilities|Casting|CastToLandscape FireApplicationTarget
          (:then
            (LocalVariables|SetTargetedLandscape true))
          (:CastFailed
            _))
        (Utilities|IsValid (Variables|Default|GetLastDamageInstigator)
          (:Is%20Valid
            (Variables|Default|SetFireInstigator (Fortnite|GetFortPlayerControllerfromActor (Variables|Default|GetLastDamageInstigator))))
          (:Is%20Not%20Valid
            (Variables|Default|SetFireInstigator _self)))
        (Curie|ExecuteFireApplication (Variables|Default|GetFireInstigator) _self FireApplicationTarget 1e+02 (Collision|MakeHitResult false false 0.0 0.0 (Transformation|GetActorLocation (select (LocalVariables|GetTargetedLandscape) self FireApplicationTarget)) (Transformation|GetActorLocation (select (LocalVariables|GetTargetedLandscape) self FireApplicationTarget)) "0,0,1" "0,0,1" 0 FireApplicationTarget) -1 (ScalableFloat|GetValueatLevel (Variables|Default|GetRadius)))))))

(fn LaunchActor
  (params (Actor%20To%20Launch :type Actor%20Object%20Reference))
  (bind _actor_to_launch (Variables|GetActor%20To%20Launch))
  (bind _asfort_pawn (Utilities|Casting|CastToFortPawn Actor%20To%20Launch)
    (:then
      ;[1 1 1] Launch Player
      (if (not (or (GameplayTags|HasMatchingGameplayTag _asfort_pawn "(TagName=\"Gameplay.Status.KnockbackImmunity\")") (GameplayTags|HasMatchingGameplayTag _asfort_pawn "(TagName=\"Gameplay.Action.Player.AthenaInVehicle\")")))
        (Character|LaunchCharacter _asfort_pawn (Utilities|CalculateLaunchVelocity (Transformation|GetActorLocation (Variables|GetActor%20To%20Launch)) (Transformation|GetActorLocation)) true true)))
    (:CastFailed
      ;[1 1 1] LaunchVehicle
      (bind _asfort_vehicle_interface (Utilities|Casting|CastToFortVehicleInterface _actor_to_launch))
      (bind _returnvalue (Vehicle|GetVehicleActor _asfort_vehicle_interface :!class /Script/FortniteGame.FortVehicleInterface))
      (Audio|BroadcastAbilityImpactatLocation 0 _returnvalue (Math|Vector|MakeVector (.x (* (Math|Vector|Normalize (select (>= (Math|Vector|VectorLength (Transformation|GetVelocity _returnvalue)) 5e+02) (Transformation|GetVelocity _returnvalue) (- (Transformation|GetActorLocation _actor_to_launch) (Transformation|GetActorLocation)))) (ScalableFloat|GetValueatLevel (Variables|Default|GetVehicleImpulseStrength)))) (.y (* (Math|Vector|Normalize (select (>= (Math|Vector|VectorLength (Transformation|GetVelocity _returnvalue)) 5e+02) (Transformation|GetVelocity _returnvalue) (- (Transformation|GetActorLocation _actor_to_launch) (Transformation|GetActorLocation)))) (ScalableFloat|GetValueatLevel (Variables|Default|GetVehicleImpulseStrength)))) (ScalableFloat|GetValueatLevel (Variables|Default|GetVehicleImpulseStrengthZ))) 777.0))))

(fn MinigameIsInProgress
  (params (IgnoreWarmUp :type Boolean))
  (bind _self self)
  (bind _returnvalue (Minigame|GetMinigame _self :!class /Script/FortniteGame.MinigameBlueprintLibrary))
  (Utilities|IsValid _returnvalue
    (:Is%20Valid
      (Utilities|IsValid _returnvalue
        (:Is%20Valid
          ;[1 1 1] Is in progress pure function includes Warmup time, optionally ignore that and only use "In Progress" time
          (if IgnoreWarmUp
            (return true (Utilities|Enum|Equal(Enum) (Variables|/Script/FortniteGame.FortMinigame|GetCurrentState _returnvalue) "InProgress"))
            (else
              (return true (Minigame|IsinProgress _returnvalue)))))
        (:Is%20Not%20Valid
          _)))
    (:Is%20Not%20Valid
      _))
  (return false false))

(fn OnRep_DamageStateNotify
  ;[1 1 1] Use Current Health to figure out which emitters to spawn
  (bind (_array_element _array_index) (Utilities|Array|ForEachLoopwithBreak (Variables|DamageBasedEmitters|GetDamageStateParticleSystems))
    (:LoopBody
      (if (.Spawned_19_ADE1ECD74E521455DA9695B1FDB47581 (Utilities|Struct|BreakAthenaDamageStateParticles _array_element))
        _
        (elif (<= (Health|GetHealthPercent) (.HealthPercent_24_AC35B253419518DA53906DB5CCAD79F8 (Utilities|Struct|BreakAthenaDamageStateParticles _array_element)))
          ;[1 1 1] Set that this element has spawned (to avoid doing it twice)
          (Utilities|Array|SetArrayElem (Variables|DamageBasedEmitters|GetDamageStateParticleSystems) _array_index (Utilities|Struct|MakeAthenaDamageStateParticles (Utilities|Struct|BreakAthenaDamageStateParticles _array_element) true (.LocalSpawnPosition_9_507DFDCB4A99E944783EEAB44C495149 (Utilities|Struct|BreakAthenaDamageStateParticles _array_element)) (.LocalSpawnRotation_12_86D59DAF4DC65EF2E0CEB2BD7740F557 (Utilities|Struct|BreakAthenaDamageStateParticles _array_element)) (.LocalSpawnScale_20_D001987D4D7EB90DF56AAC9B5F8133A8 (Utilities|Struct|BreakAthenaDamageStateParticles _array_element))))
          ;[1 1 1] Spawn the effect
          (bind _returnvalue (AddComponent|NODE_AddParticleSystemComponent-0 :RelativeTransform_Location (.LocalSpawnPosition_9_507DFDCB4A99E944783EEAB44C495149 (Utilities|Struct|BreakAthenaDamageStateParticles _array_element)) :RelativeTransform_Rotation (.LocalSpawnRotation_12_86D59DAF4DC65EF2E0CEB2BD7740F557 (Utilities|Struct|BreakAthenaDamageStateParticles _array_element)) :RelativeTransform_Scale (.LocalSpawnScale_20_D001987D4D7EB90DF56AAC9B5F8133A8 (Utilities|Struct|BreakAthenaDamageStateParticles _array_element))))
          (Effects|Components|ParticleSystem|SetTemplate _returnvalue (Utilities|Struct|BreakAthenaDamageStateParticles _array_element))
          (Utilities|Array|Add (Variables|DamageBasedEmitters|GetSpawnedDamageStateEmitters) _returnvalue))))))

(fn OnRep_ExplosiveFxRepNotify
  (if (not (Variables|Default|GetDestroyedByPickAxe))
    (Projectile|RemoveGameplayCue%28Looping%29Local self (Variables|Default|GetFireCue) (Ability|GameplayCue|MakeGameplayCueParameters 0.0))
    (GameplayCue|ExecuteGameplayCue%28Burst%29Local self (Variables|GameplayCues|GetExplosionCUE) (Ability|GameplayCue|MakeGameplayCueParameters 0.0) false)))

(fn ShouldCancelExplosion
  ;[1 0.08217 0.06227] Don't allow the prop to explode if it is in a valid minigame that is not running.
  (bind (_hasvalidminigame _isrunning) (Default|MinigameIsinProgress false))
  (return (and _hasvalidminigame (not _isrunning))))

(fn ShouldDie
  (params (Damage :type Float%20%28single-precision%29) (EventInstigator :type Controller%20Object%20Reference) (DamageCauser :type Actor%20Object%20Reference) (bIsTest :type Boolean) (EffectContext :type Gameplay%20Effect%20Context%20Handle%20Structure%20%28by%20ref%29))
  ;[1 1 1] This Function get's called on Death. Health Is Zero get's set to true so that the damage event triggers the Death Fx. Should die is set to false so that the actors collision stays intact
  (Variables|Bools|SetHealthIsZero (Math|Float|NearlyEqual(Float) (Health|GetHealthPercent) 0.0 0.001))
  (return (Variables|Default|GetbShouldDie)))

(fn ConstructionScript
  (|CallParentFunction))
